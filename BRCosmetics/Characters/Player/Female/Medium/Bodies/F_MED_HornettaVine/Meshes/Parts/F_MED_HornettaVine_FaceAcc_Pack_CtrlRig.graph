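# Control Rig for the HornettaVine female face-accessory pack that partially aims the ponytail at a helper bone on the back.
# From Begin Execution, it sets the local transform of aim_ponytail, a child of spine_03, then aims dyn_ponytail_03 toward it at half primary weight with no secondary axis.
# Separate aim, parent and position constraint chains linking lower ponytail bones to neck and spine are left unconnected to execution.
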ControlRigBlueprint /BRCosmetics/Characters/Player/Female/Medium/Bodies/F_MED_HornettaVine/Meshes/Parts/F_MED_HornettaVine_FaceAcc_Pack_CtrlRig
graphs [
 RigVMFunctionLibrary {links [], nodes []}
 RigVMModel {nodes [
   &RigUnit_AimConstraintLocalSpaceOffset(
    AdvancedSettings=(DebugSettings=(Scale=10,WorldOffset=(Rotation=(X=0,Y=0,Z=0,W=1),Translation=(X=0,Y=0,Z=0),Scale3D=(X=1,Y=1,Z=1))),RotationOrderForFilter=XZY)
    Child=(Type=Bone,Name="dyn_ponytail_03")
    ExecutePin=()
    Filter=(bX=True,bY=True,bZ=True)
    Parents=((Item=(Type=Bone,Name="spine_05"),Weight=1))
    Weight=1
    WorldUp=(Target=(X=0,Y=0,Z=1),Space=(Type=Bone,Name="aim_ponytail"))
   )
   > &RigUnit_AimConstraintLocalSpaceOffset(
    AdvancedSettings=(DebugSettings=(Scale=10,WorldOffset=(Rotation=(X=0,Y=0,Z=0,W=1),Translation=(X=0,Y=0,Z=0),Scale3D=(X=1,Y=1,Z=1))),RotationOrderForFilter=XZY)
    Child=(Type=Bone,Name="dyn_ponytail_04")
    Filter=(bX=True,bY=True,bZ=True)
    Parents=((Item=(Type=Bone,Name="spine_04"),Weight=1))
    Weight=1
    WorldUp=(Target=(X=0,Y=0,Z=1),Space=(Type=Bone,Name="spine_03"))
   )
   > &RigUnit_AimConstraintLocalSpaceOffset(
    AdvancedSettings=(DebugSettings=(Scale=10,WorldOffset=(Rotation=(X=0,Y=0,Z=0,W=1),Translation=(X=0,Y=0,Z=0),Scale3D=(X=1,Y=1,Z=1))),RotationOrderForFilter=XZY)
    Child=(Type=Bone,Name="dyn_ponytail_05")
    Filter=(bX=True,bY=True,bZ=True)
    Parents=((Item=(Type=Bone,Name="spine_03"),Weight=1))
    Weight=1
    WorldUp=(Target=(X=0,Y=0,Z=1),Space=(Type=Bone,Name="spine_05"))
   )
   &RigUnit_ParentConstraint(
    AdvancedSettings=(RotationOrderForFilter=XZY)
    Child=(Type=Bone,Name="dyn_ponytail_03")
    ExecutePin=()
    Parents=((Item=(Type=Bone,Name="neck_01"),Weight=1))
    Weight=1
   )
   > &RigUnit_ParentConstraint(
    AdvancedSettings=(RotationOrderForFilter=XZY)
    Child=(Type=Bone,Name="dyn_ponytail_04")
    Parents=((Item=(Type=Bone,Name="spine_05"),Weight=1))
    Weight=1
   )
   > &RigUnit_ParentConstraint(
    AdvancedSettings=(RotationOrderForFilter=XZY)
    Child=(Type=Bone,Name="dyn_ponytail_05")
    Parents=((Item=(Type=Bone,Name="spine_03"),Weight=1))
    Weight=1
   )
   /Script/ControlRig.RigUnit_PositionConstraintLocalSpaceOffset(
    Child=(Type=Bone,Name="dyn_ponytail_03")
    ExecutePin=()
    Filter=(bX=True,bY=True,bZ=True)
    Parents=((Item=(Type=Bone,Name="neck_01"),Weight=1))
    Weight=1
    bMaintainOffset=True
   ) {is=struct RigVMUnitNode}
   > &RigUnit_PositionConstraintLocalSpaceOffset(
    Child=(Type=Bone,Name="dyn_ponytail_04")
    Filter=(bX=True,bY=True,bZ=True)
    Parents=((Item=(Type=Bone,Name="spine_04"),Weight=1))
    Weight=1
    bMaintainOffset=True
   )
   > &RigUnit_PositionConstraintLocalSpaceOffset(
    Child=(Type=Bone,Name="dyn_ponytail_05")
    Filter=(bX=True,bY=True,bZ=True)
    Parents=((Item=(Type=Bone,Name="spine_03"),Weight=1))
    Weight=1
    bMaintainOffset=True
   )
   RigUnit_BeginExecution = &RigUnit_BeginExecution()
   &RigUnit_SetTransform(
    Item=(Type=Bone,Name="aim_ponytail")
    Space=LocalSpace
    Value=(Rotation=(X=0,Y=0,Z=0,W=1),Translation=(X=0,Y=-2,Z=0),Scale3D=(X=1,Y=1,Z=1))
    Weight=1
    bPropagateToChildren=true
    ExecutePin=@RigUnit_BeginExecution.ExecutePin
   )
   > /Script/ControlRig.RigUnit_AimItem(
    DebugSettings=(Scale=10,WorldOffset=(Rotation=(X=0,Y=0,Z=0,W=1),Translation=(X=0,Y=0,Z=0),Scale3D=(X=1,Y=1,Z=1)))
    Item=(Type=Bone,Name="dyn_ponytail_03")
    Primary=(Weight=0.5,Axis=(X=1,Y=0,Z=0),Target=(X=0,Y=-2,Z=0),Kind=Location,Space=(Type=Bone,Name="aim_ponytail"))
    Secondary=(Weight=0,Axis=(X=0,Y=0,Z=0),Target=(X=1,Y=0,Z=0),Kind=Direction,Space=(Type=Bone,Name="head"))
    Weight=1
   ) {is=struct RigVMUnitNode}
  ]}
]
hierarchy [
 root {loc=<0 0 0>, rot=<0 0 0 1>}
 pelvis {parent=root, loc=<0 -0.5301 98.0571>, rot=<0.0127 0.707 -0.0127 -0.707>}
 spine_01 {parent=pelvis, loc=<6.1208 -0.9328 0>, rot=<0 0 0.034 -0.9994>}
 spine_02 {parent=spine_01, loc=<6.5996 -0.2095 0>, rot=<0 0 0.0238 -0.9997>}
 spine_03 {parent=spine_02, loc=<6.4914 -0.1237 0>, rot=<0 0 0.0149 -0.9999>}
 spine_04 {parent=spine_03, loc=<6.9598 0.0794 0>, rot=<0 0 -0.0132 -0.9999>}
 spine_05 {parent=spine_04, loc=<6.9987 0 0>, rot=<0 0 -0.0015 -1>}
 neck_01 {parent=spine_05, loc=<11.3138 0.1529 0>, rot=<0 0 -0.2003 -0.9797>}
 neck_02 {parent=neck_01, loc=<4.5627 0 0>, rot=<0 0 0.0504 -0.9987>}
 head {parent=neck_02, loc=<5.3914 0 0>, rot=<0 0 0.115 -0.9934>}
 dyn_head_offset {parent=head, loc=<6.7447 0 0>, rot=<0 0 0 1>}
 dyn_bangs_base {parent=dyn_head_offset, loc=<3.7969 6.3637 3.4089>, rot=<-0.67 0.1274 -0.1425 -0.7173>}
 dyn_bangs_parent {parent=dyn_bangs_base, loc=<0 0 0>, rot=<0 0 0 1>}
 dyn_bangs_01 {parent=dyn_bangs_parent, loc=<3.3998 0.4075 -2.9667>, rot=<0.1204 -0.161 0.9796 -0.0038>}
 dyn_bangs_02 {parent=dyn_bangs_01, loc=<5.3942 -3.8357 -1.8604>, rot=<-0.0409 -0.5014 0.3476 -0.7912>}
 dyn_ponytail_base {parent=dyn_head_offset, loc=<11.2755 -12.1458 0.3199>, rot=<-0.1619 0.6883 0.6883 -0.1619>}
 dyn_ponytail_parent {parent=dyn_ponytail_base, loc=<0 0 0>, rot=<0 0 0 1>}
 dyn_ponytail_01 {parent=dyn_ponytail_parent, loc=<6.1009 0.3487 1.3688>, rot=<0 0.225 0 -0.9744>}
 dyn_ponytail_02 {parent=dyn_ponytail_01, loc=<9.0125 0.118 0.0109>, rot=<0 0 0 1>}
 dyn_ponytail_03 {parent=dyn_ponytail_02, loc=<11.8348 -0.0577 0.9333>, rot=<0 0 0 1>}
 dyn_ponytail_04 {parent=dyn_ponytail_03, loc=<13.9477 -0.2996 0.2521>, rot=<0 -0.0175 0 -0.9998>}
 dyn_ponytail_05 {parent=dyn_ponytail_04, loc=<11.8635 -1.0103 -1.4307>, rot=<0 -0.1219 0 -0.9925>}
 clavicle_l {parent=spine_05, loc=<5.4831 -0.1053 -3.2946>, rot=<0.0465 -0.7289 -0.009 -0.6829>}
 upperarm_l {parent=clavicle_l, loc=<9.8658 -1.2136 -0.4105>, rot=<0.0164 -0.3245 -0.0269 -0.9454>}
 lowerarm_l {parent=upperarm_l, loc=<26.465 -0.8345 0.259>, rot=<0.0036 0.0161 -0.2298 -0.9731>}
 upperarm_twist_02_l {parent=upperarm_l, loc=<19.7998 -0.1134 0>, rot=<0 0 0 1>}
 clavicle_r {parent=spine_05, loc=<5.4831 -0.1053 3.2946>, rot=<0.7289 0.0465 -0.6829 0.009>}
 upperarm_r {parent=clavicle_r, loc=<-9.8658 1.2136 0.4105>, rot=<0.0164 -0.3245 -0.0269 -0.9454>}
 lowerarm_r {parent=upperarm_r, loc=<-26.465 0.8345 -0.259>, rot=<0.0036 0.0161 -0.2298 -0.9731>}
 upperarm_twist_02_r {parent=upperarm_r, loc=<-19.7998 0.1134 0>, rot=<0 0 0 -1>}
 aim_ponytail {parent=spine_03, loc=<1.067 -5.5051 0.2531>, rot=<0.706 0.0387 -0.0387 -0.706>}
]
