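# Physics asset for the HornettaVine female medium FaceAcc pack mesh, defining bodies and joint limits for its dynamic ponytail and bangs.
# Two constraint chains run from small base spheres through capsule segments, with tight swing and twist limits, slightly looser on the bangs.
# Unconstrained neck, spine, pelvis and upper-arm shapes approximate the body, likely as colliders for the hair.
# The arm shapes are asymmetric: only the right side has an upper-arm twist capsule.
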PhysicsAsset /BRCosmetics/Characters/Player/Female/Medium/Bodies/F_MED_HornettaVine/Meshes/Parts/F_MED_HornettaVine_FaceAcc_Pack_Physics
bodies {
 dyn_bangs_01 [{shapeType=Capsule, length=3.1627, radius=0.871, shapeName=dyn_bangs_01_capsule, center=<1.652 0 0>, rotation=<rot 90 0 0>}]
 dyn_bangs_02 [{shapeType=Capsule, length=1.4039, radius=1.8584, shapeName=dyn_bangs_02_capsule, center=<1.2197 0 0>, rotation=<rot 90 0 0>}]
 dyn_bangs_base [{shapeType=Sphere, radius=0.5, shapeName=dyn_bangs_base_sphere}]
 dyn_bangs_parent [{shapeType=Capsule, length=2.6338, radius=0.7877, shapeName=dyn_bangs_parent_capsule, center=<1.4134 0 0>, rotation=<rot 90 0 0>}]
 dyn_ponytail_01 [{shapeType=Capsule, length=6.8478, radius=2.6478, shapeName=dyn_ponytail_01_capsule, center=<4.1976 0 -0.3637>, rotation=<rot 87.184 0 0>}]
 dyn_ponytail_02 [{
   shapeType=Capsule
   length=6.4121
   radius=2.9177
   shapeName=dyn_ponytail_02_capsule
   center=<4.7065 0 0.2882>
   rotation=<rot 86.0633 -180 -180>
  }]
 dyn_ponytail_03 [{shapeType=Capsule, length=9.5503, radius=2.8403, shapeName=dyn_ponytail_03_capsule, center=<5.3454 0 0>, rotation=<rot 90 0 0>}]
 dyn_ponytail_04 [{shapeType=Capsule, length=8.7969, radius=2.9369, shapeName=dyn_ponytail_04_capsule, center=<5.2823 0 -0.5084>, rotation=<rot 82.9168 0 0>}]
 dyn_ponytail_05 [{shapeType=Capsule, length=7.3015, radius=2.8919, shapeName=dyn_ponytail_05_capsule, center=<5.3454 0 0>, rotation=<rot 90 0 0>}]
 dyn_ponytail_base [{shapeType=Sphere, radius=0.5, shapeName=dyn_ponytail_base_sphere}]
 dyn_ponytail_parent [{shapeType=Capsule, length=1.8186, radius=3.1893, shapeName=dyn_ponytail_parent_capsule, center=<2.8983 0.3581 0>, rotation=<rot 90 0 0>}]
 neck_02 [{shapeType=Capsule, length=5.211, radius=5.8759, shapeName=neck_02_capsule, center=<1.9564 0 0>, rotation=<rot 90 0 0>}]
 pelvis [{shapeType=Capsule, length=16.8536, radius=13.1464, shapeName=pelvis_capsule, center=<0 -1.8674 0>}]
 spine_04 [{
   shapeType=Capsule
   length=12.4539
   radius=13.2427
   shapeName=spine_04_capsule
   center=<-9.5215 4.9427 0>
   rotation=<rot 90 -179.5524 -178.5568>
  }]
 spine_05 [
  {shapeType=Capsule, length=15.0827, radius=12.1166, shapeName=spine_05_capsule, center=<0 4.5163 0>}
  {shapeType=Capsule, length=48.1342, radius=18.4927, shapeName=spine_05_capsule, center=<23.2475 15.2212 0>}
 ]
 upperarm_l [{shapeType=Sphere, radius=8.4892, shapeName=upperarm_l_sphere, center=<-1.7298 2.2422 3.8703>}]
 upperarm_r [{shapeType=Sphere, radius=8.4892, shapeName=upperarm_r_sphere, center=<1.6316 -3.6774 -3.3478>}]
 upperarm_twist_02_r [{shapeType=Capsule, length=23.5625, radius=6.4375, shapeName=upperarm_twist_02_r_capsule, center=<4.871 0 0>, rotation=<rot 90 0 0>}]
}
constraints [
 {bone1Name=dyn_ponytail_parent, bone2Name=dyn_ponytail_base, swing1LimitDegrees=5, swing2LimitDegrees=5, twistLimitDegrees=3}
 {bone1Name=dyn_ponytail_01, bone2Name=dyn_ponytail_parent, swing1LimitDegrees=5, swing2LimitDegrees=3, twistLimitDegrees=2}
 {bone1Name=dyn_ponytail_02, bone2Name=dyn_ponytail_01, swing1LimitDegrees=3, swing2LimitDegrees=3, twistLimitDegrees=2}
 {bone1Name=dyn_ponytail_03, bone2Name=dyn_ponytail_02, swing1LimitDegrees=5, swing2LimitDegrees=3, twistLimitDegrees=3}
 {bone1Name=dyn_ponytail_04, bone2Name=dyn_ponytail_03, swing1LimitDegrees=5, swing2LimitDegrees=3, twistLimitDegrees=3}
 {bone1Name=dyn_ponytail_05, bone2Name=dyn_ponytail_04, swing1LimitDegrees=5, swing2LimitDegrees=3, twistLimitDegrees=3}
 {bone1Name=dyn_bangs_parent, bone2Name=dyn_bangs_base, swing1LimitDegrees=10, swing2LimitDegrees=8, twistLimitDegrees=2}
 {bone1Name=dyn_bangs_01, bone2Name=dyn_bangs_parent, swing1LimitDegrees=5, swing2LimitDegrees=10, twistLimitDegrees=2}
 {bone1Name=dyn_bangs_02, bone2Name=dyn_bangs_01, swing1LimitDegrees=8, swing2LimitDegrees=12, twistLimitDegrees=2}
]
