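# Physics asset for the HornettaVine female character's FaceAcc mesh part, defining physics bodies for ponytail and bangs bone chains.
# Each chain runs from a small base sphere through capsules linked parent to child by constraints with tight swing and twist limits.
# Unconstrained shapes on the head, neck, upper spine, pelvis and upper arms form a rough body volume around the hair.
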PhysicsAsset /BRCosmetics/Characters/Player/Female/Medium/Bodies/F_MED_HornettaVine/Meshes/Parts/F_MED_HornettaVine_FaceAcc_Physics
bodies {
 dyn_bangs_01 [{shapeType=Capsule, length=3.1627, radius=0.871, shapeName=dyn_bangs_01_capsule, center=<2.267 0 0>, rotation=<rot 90 0 0>}]
 dyn_bangs_02 [{shapeType=Capsule, length=2.2164, radius=1.0459, shapeName=dyn_bangs_02_capsule, center=<1.822 0 0>, rotation=<rot 90 0 0>}]
 dyn_bangs_base [{shapeType=Sphere, radius=0.5, shapeName=dyn_bangs_base_sphere}]
 dyn_bangs_parent [{shapeType=Capsule, length=2.6338, radius=0.7877, shapeName=dyn_bangs_parent_capsule, center=<1.7556 0 0>, rotation=<rot 90 0 0>}]
 dyn_ponytail_01 [{shapeType=Capsule, length=4.7772, radius=3.1485, shapeName=dyn_ponytail_01_capsule, center=<4.1976 0 0>, rotation=<rot 90 0 0>}]
 dyn_ponytail_02 [{
   shapeType=Capsule
   length=8.4362
   radius=2.962
   shapeName=dyn_ponytail_02_capsule
   center=<6.1795 0 0.5619>
   rotation=<rot 85.8825 -180 -180>
  }]
 dyn_ponytail_03 [{shapeType=Capsule, length=11.0488, radius=2.4704, shapeName=dyn_ponytail_03_capsule, center=<6.7317 0 0>, rotation=<rot 90 0 0>}]
 dyn_ponytail_04 [{shapeType=Capsule, length=9.4219, radius=2.3119, shapeName=dyn_ponytail_04_capsule, center=<6.0939 0 -0.5335>, rotation=<rot 83.0424 0 0>}]
 dyn_ponytail_05 [{shapeType=Capsule, length=7.3224, radius=2.871, shapeName=dyn_ponytail_05_capsule, center=<5.3454 0 0.2019>, rotation=<rot 90 0 0>}]
 dyn_ponytail_base [{shapeType=Sphere, radius=0.5, shapeName=dyn_ponytail_base_sphere}]
 dyn_ponytail_parent [{shapeType=Capsule, length=2.8272, radius=2.9115, shapeName=dyn_ponytail_parent_capsule, center=<2.8983 0.3581 0>, rotation=<rot 90 0 0>}]
 head [{shapeType=Sphere, radius=0.5, shapeName=head_sphere}]
 neck_02 [{shapeType=Capsule, length=5.211, radius=5.8759, shapeName=neck_02_capsule, center=<1.9564 0 0>, rotation=<rot 90 0 0>}]
 pelvis [{shapeType=Capsule, length=15, radius=15, shapeName=pelvis_capsule, center=<-4.7191 -1.0141 0>}]
 spine_04 [{shapeType=Capsule, length=8.3788, radius=12.6588, shapeName=spine_04_capsule, center=<-9.5337 3.069 0>, rotation=<rot 90 0 0>}]
 spine_05 [
  {shapeType=Capsule, length=15.0827, radius=12.1166, shapeName=spine_05_capsule, center=<0 4.5163 0>}
  {shapeType=Capsule, length=48.1342, radius=18.4927, shapeName=spine_05_capsule, center=<23.2475 15.2212 0>}
 ]
 upperarm_l [{shapeType=Sphere, radius=8.4892, shapeName=upperarm_l_sphere, center=<-1.7298 2.2422 3.8703>}]
 upperarm_r [{shapeType=Sphere, radius=8.4892, shapeName=upperarm_r_sphere, center=<1.6316 -3.6774 -3.3478>}]
 upperarm_twist_02_l [{shapeType=Capsule, length=21.8098, radius=8.1902, shapeName=upperarm_twist_02_l_capsule, center=<-6.1007 0 0>, rotation=<rot 90 0 0>}]
 upperarm_twist_02_r [{shapeType=Capsule, length=21.8098, radius=8.1902, shapeName=upperarm_twist_02_r_capsule, center=<5.5403 0 0>, rotation=<rot 90 0 0>}]
}
constraints [
 {bone1Name=dyn_ponytail_parent, bone2Name=dyn_ponytail_base, swing1LimitDegrees=10, swing2LimitDegrees=15, twistLimitDegrees=3}
 {bone1Name=dyn_ponytail_01, bone2Name=dyn_ponytail_parent, swing1LimitDegrees=10, swing2LimitDegrees=15, twistLimitDegrees=3}
 {bone1Name=dyn_ponytail_02, bone2Name=dyn_ponytail_01, swing1LimitDegrees=10, swing2LimitDegrees=10, twistLimitDegrees=3}
 {bone1Name=dyn_ponytail_03, bone2Name=dyn_ponytail_02, swing1LimitDegrees=10, swing2LimitDegrees=10, twistLimitDegrees=3}
 {bone1Name=dyn_ponytail_04, bone2Name=dyn_ponytail_03, swing1LimitDegrees=10, swing2LimitDegrees=10, twistLimitDegrees=3}
 {bone1Name=dyn_ponytail_05, bone2Name=dyn_ponytail_04, swing1LimitDegrees=10, swing2LimitDegrees=10, twistLimitDegrees=3}
 {bone1Name=dyn_bangs_parent, bone2Name=dyn_bangs_base, swing1LimitDegrees=10, swing2LimitDegrees=8, twistLimitDegrees=2}
 {bone1Name=dyn_bangs_01, bone2Name=dyn_bangs_parent, swing1LimitDegrees=5, swing2LimitDegrees=10, twistLimitDegrees=2}
 {bone1Name=dyn_bangs_02, bone2Name=dyn_bangs_01, swing1LimitDegrees=8, swing2LimitDegrees=12, twistLimitDegrees=2}
]
