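# Anim Blueprint for the HornettaVine face accessory that drives its ponytail secondary motion from the copied body pose.
# Backbling state selects simple RigidBody physics without a pack, Control Rig, spine constraints and pack physics for short packs, or head-space physics with additive base tilt for long ones.
# The long-pack tilt increases, interpolated, when forward velocity passes a threshold.
# A deprecated state function supplies skydive-aware gravity to the non-pack and long-pack simulations.
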AnimBlueprint /BRCosmetics/Characters/Player/Female/Medium/Bodies/F_MED_HornettaVine/Meshes/Parts/F_MED_HornettaVine_FaceAcc_AnimBP
target_skeleton=&/F_MED_HornettaVine_FaceAcc_Skeleton
graphs [
 AnimGraph {nodes [
   "Constraint - dyn_ponytail_03 (P)"!(ComponentPose, Alpha=0.75, ConstraintWeights_0=1, ConstraintWeights_1:float=0) -> (Pose) {
    alphaBoolBlend {blendOption=Linear}
    boneToModify {boneName=dyn_ponytail_03}
    constraintSetup [{targetBone {boneName=spine_05}}]
    constraintWeights [1,0]
   }
   Constraint!(
    ComponentPose
    Alpha=1
    ConstraintWeights_0=0.3426
    ConstraintWeights_1:float=0.1394
    ConstraintWeights_2:float=0.2922
    ConstraintWeights_3:float=0.1652
    ConstraintWeights_4:float=0.0606
   ) -> (Pose) {
    alphaBoolBlend {blendOption=Linear}
    constraintSetup [
     {targetBone {boneName=spine_02}}
     {targetBone {boneName=spine_04}}
     {targetBone {boneName=spine_03}}
     {targetBone {boneName=spine_01}}
     {targetBone {boneName=pelvis}}
    ]
    constraintWeights [1,1,1,1,1]
    boneToModify=~
   }
   "Constraint - dyn_ponytail_04 (PPPP)"!(
    ComponentPose
    Alpha=0.58
    ConstraintWeights_0=0.479
    ConstraintWeights_1:float=0.308
    ConstraintWeights_2:float=0.1065
    ConstraintWeights_3:float=0.1065
   ) -> (Pose) {
    alphaBoolBlend {blendOption=Linear}
    boneToModify {boneName=dyn_ponytail_04}
    constraintSetup [{targetBone {boneName=spine_05}}, {targetBone {boneName=spine_04}}, {targetBone {boneName=clavicle_r}}, {targetBone {boneName=clavicle_l}}]
    constraintWeights [1,1,1,1]
   }
   "Constraint - dyn_ponytail_04 (PPPPPPP)"!(
    ComponentPose
    Alpha=0.5
    ConstraintWeights_0=0
    ConstraintWeights_1:float=0.5
    ConstraintWeights_2:float=0.2
    ConstraintWeights_3:float=0.3
    ConstraintWeights_4:float=0
    ConstraintWeights_5:float=0
    ConstraintWeights_6:float=0
   ) -> (Pose) {
    alphaBoolBlend {blendOption=Linear}
    boneToModify {boneName=dyn_ponytail_04}
    constraintSetup [
     {targetBone {boneName=spine_02}}
     {targetBone {boneName=spine_04}}
     {targetBone {boneName=spine_03}}
     {targetBone {boneName=spine_05}}
     {targetBone {boneName=clavicle_l}}
     {targetBone {boneName=clavicle_r}}
     {targetBone {boneName=spine_01}}
    ]
    constraintWeights [1,1,1,1,1,1,1]
   }
   Control Rig() {
    alphaBoolBlend {blendOption=Linear}
    controlRigAssetReference {blueprintRigClass=&/F_MED_HornettaVine_FaceAcc_Pack_CtrlRig.F_MED_HornettaVine_FaceAcc_Pack_CtrlRig_C}
   }
   n1 = Look At - Bone: dyn_ponytail_03(Alpha=1) {
    alphaBoolBlend {blendOption=Linear}
    boneToModify {boneName=dyn_ponytail_03}
    lookAtLocation=<0 100>
    lookAtTarget {boneReference {boneName=neck_01}}
    lookAt_Axis {axis {y=0, z=1}}
   }
   n2 = Look At - Bone: dyn_ponytail_04(ComponentPose=@n1, Alpha=1) {
    alphaBoolBlend {blendOption=Linear}
    boneToModify {boneName=dyn_ponytail_04}
    lookAtLocation=<0 100>
    lookAtTarget {boneReference {boneName=spine_05}}
    lookAt_Axis {axis {y=0, z=1}}
   }
   Look At - Bone: dyn_ponytail_05(ComponentPose=@n2, Alpha=1) {
    alphaBoolBlend {blendOption=Linear}
    boneToModify {boneName=dyn_ponytail_05}
    lookAtLocation=<0 100>
    lookAtTarget {boneReference {boneName=spine_03}}
    lookAt_Axis {axis {y=0, z=1}}
   }
   "Transform (Modify) Bone - Bone: dyn_ponytail_03"(Alpha=1, Translation="0,0,0", Rotation="-40,0,0", Scale="1,1,1") {
    rotationMode=BMM_Additive
    rotationSpace=BCS_BoneSpace
    alphaBoolBlend {blendOption=Linear}
    boneToModify {boneName=dyn_ponytail_03}
    rotation {pitch=-40}
   }
   n3 = RigidBody!(
    ComponentPose=Local To Component(LocalPose=Use cached pose 'PackCache'())
    Alpha
    OverrideWorldGravity:Vector=Get Gravity_Override() -> (Gravity_Override:Vector)
    ExternalForce
   ) -> (Pose) {
    bClampLinearTranslationLimitToRefPose=true
    bOverrideWorldGravity=true
    overridePhysicsAsset=&/F_MED_HornettaVine_FaceAcc_Physics
    simulationSpace=BaseBoneSpace
    alphaBoolBlend {blendOption=Linear}
    baseBoneRef {boneName=head}
    componentAppliedLinearAccClamp=<6000 6000 6000>
    componentLinearAccScale=<0.2 0.2 0.258>
    componentLinearVelScale=<0.2 0.2 0.25>
    overrideWorldGravity {z=-980}
    simSpaceSettings {
     maxAngularAcceleration=500
     maxAngularVelocity=3
     maxLinearAcceleration=400
     maxLinearVelocity=250
     worldAlpha=1
     externalLinearDragV=<0.1 0.1>
    }
   }
   n4 = "Transform (Modify) Bone - Bone: dyn_ponytail_base"(ComponentPose=@n3, Alpha=Get AcceleratingForward() -> (AcceleratingForward:double), Translation="0,0,0", Rotation="-20,0,0", Scale="1,1,1") {
    rotationMode=BMM_Additive
    rotationSpace=BCS_BoneSpace
    alphaBoolBlend {blendOption=Linear}
    alphaScaleBiasClamp {bInterpResult=true, interpSpeedDecreasing=4, interpSpeedIncreasing=6}
    boneToModify {boneName=dyn_ponytail_base}
    rotation {pitch=-20}
   }
   n5 = "Transform (Modify) Bone - Bone: dyn_ponytail_base"(ComponentPose=@n4, Alpha=1, Translation="0,0,0", Rotation="-15,0,0", Scale="1,1,1") {
    rotationMode=BMM_Additive
    rotationSpace=BCS_BoneSpace
    alphaBoolBlend {blendOption=Linear}
    boneToModify {boneName=dyn_ponytail_base}
    rotation {pitch=-15}
   }
   n6 = Component To Local(ComponentPose=@n5)
   n7 = Control Rig(Source=Use cached pose 'PackCache'()) {
    alphaBoolBlend {blendOption=Linear}
    controlRigAssetReference {blueprintRigClass=&F_MED_HornettaVine_FaceAcc_Pack_CtrlRig_C}
   }
   n8 = "Constraint - dyn_ponytail_02 (PPPP)"!(
    ComponentPose=Local To Component(LocalPose=@n7)
    Alpha=1
    ConstraintWeights_0=0.1
    ConstraintWeights_1:float=0.1
    ConstraintWeights_2:float=0.6
    ConstraintWeights_3:float=0.2
   ) -> (Pose) {
    alphaBoolBlend {blendOption=Linear}
    boneToModify {boneName=dyn_ponytail_02}
    constraintSetup [{targetBone {boneName=spine_05}}, {targetBone {boneName=neck_01}}, {targetBone {boneName=head}}, {targetBone {boneName=neck_02}}]
    constraintWeights [1,1,1,1]
   }
   n9 = "Constraint - dyn_ponytail_03 (PPPP)"!(
    ComponentPose=@n8
    Alpha=1
    ConstraintWeights_0=0.6
    ConstraintWeights_1:float=0.3
    ConstraintWeights_2:float=0.1
    ConstraintWeights_3:float=0
   ) -> (Pose) {
    alphaBoolBlend {blendOption=Linear}
    boneToModify {boneName=dyn_ponytail_03}
    constraintSetup [{targetBone {boneName=spine_05}}, {targetBone {boneName=clavicle_l}}, {targetBone {boneName=neck_01}}, {targetBone {boneName=neck_02}}]
    constraintWeights [1,1,1,1]
   }
   n10 = "Constraint - dyn_ponytail_04 (PPPPPPP)"(
    ComponentPose=@n9
    Alpha=1
    ConstraintWeights_0=0
    ConstraintWeights_1=0.1
    ConstraintWeights_2=0
    ConstraintWeights_3=0.3
    ConstraintWeights_4=0.3
    ConstraintWeights_5=0.3
    ConstraintWeights_6=0
   ) {
    alphaBoolBlend {blendOption=Linear}
    boneToModify {boneName=dyn_ponytail_04}
    constraintSetup [
     {targetBone {boneName=spine_02}}
     {targetBone {boneName=spine_04}}
     {targetBone {boneName=spine_03}}
     {targetBone {boneName=spine_05}}
     {targetBone {boneName=clavicle_l}}
     {targetBone {boneName=clavicle_r}}
     {targetBone {boneName=spine_01}}
    ]
    constraintWeights [1,1,1,1,1,1,1]
   }
   n11 = "Constraint - dyn_ponytail_05 (PPPPPPPP)"!(
    ComponentPose=@n10
    Alpha=1
    ConstraintWeights_0=0.2
    ConstraintWeights_1:float=0
    ConstraintWeights_2:float=0.2
    ConstraintWeights_3:float=0
    ConstraintWeights_4:float=0
    ConstraintWeights_5:float=0
    ConstraintWeights_6:float=0.2
    ConstraintWeights_7:float=0.4
   ) -> (Pose) {
    alphaBoolBlend {blendOption=Linear}
    boneToModify {boneName=dyn_ponytail_05}
    constraintSetup [
     {targetBone {boneName=spine_02}}
     {targetBone {boneName=spine_04}}
     {targetBone {boneName=spine_03}}
     {targetBone {boneName=spine_05}}
     {targetBone {boneName=clavicle_l}}
     {targetBone {boneName=clavicle_r}}
     {targetBone {boneName=spine_01}}
     {targetBone {boneName=pelvis}}
    ]
    constraintWeights [1,1,1,1,1,1,1,1]
   }
   n12 = RigidBody!(ComponentPose=@n11, Alpha, ExternalForce) -> (Pose) {
    bClampLinearTranslationLimitToRefPose=true
    bOverrideWorldGravity=true
    overridePhysicsAsset=&/F_MED_HornettaVine_FaceAcc_Pack_Physics
    simulationSpace=BaseBoneSpace
    alphaBoolBlend {blendOption=Linear}
    baseBoneRef {boneName=neck_02}
    componentAppliedLinearAccClamp=<6000 6000 6000>
    componentLinearAccScale=<0.1 0.1 0.1>
    componentLinearVelScale=<0.1 0.1 0.1>
    overrideWorldGravity {z=-980}
    simSpaceSettings {
     maxAngularAcceleration=100
     maxAngularVelocity=3
     maxLinearAcceleration=400
     maxLinearVelocity=250
     velocityScaleZ=0
     worldAlpha=1
     externalLinearDragV=<0.1 0.1>
    }
   }
   n13 = Component To Local(ComponentPose=@n12)
   n14 = Get isWearingCape() -> (isWearingCape:bool)
   n15 = RigidBody!(
    ComponentPose=Local To Component(LocalPose=Use cached pose 'PackCache'())
    Alpha
    OverrideWorldGravity:Vector=Get Gravity_Override()
    ExternalForce
   ) -> (Pose) {
    bClampLinearTranslationLimitToRefPose=true
    bOverrideWorldGravity=true
    overridePhysicsAsset=&F_MED_HornettaVine_FaceAcc_Physics
    simulationSpace=BaseBoneSpace
    alphaBoolBlend {blendOption=Linear}
    baseBoneRef {boneName=neck_02}
    componentAppliedLinearAccClamp=<3000 3000 6000>
    componentLinearAccScale=<0.2 0.2 0.2>
    componentLinearVelScale=<0.2 0.2 0.2>
    overrideWorldGravity {z=-980}
    simSpaceSettings {maxAngularAcceleration=100, maxAngularVelocity=8, maxLinearAcceleration=300, maxLinearVelocity=200, worldAlpha=1}
   }
   Output Pose(Result=Blend Poses by bool(
     BlendPose_0=Blend Poses by bool(BlendPose_0=@n6, BlendPose_1=@n13, bActiveValue=@n14) {blendType=Linear}
     BlendPose_1=Component To Local(ComponentPose=@n15)
     bActiveValue=Get isWearingPack() -> (isWearingPack:bool)
    ) {blendType=Linear}) {layerGroup=None}
   Save cached pose 'PackCache'(Pose=Copy Pose From Mesh(SourceMeshComponent=Get BodyMesh() -> (BodyMesh:Skeletal Mesh Component Object Reference)) {bUseAttachedParent=true}) {cacheName=PackCache}
  ]}
 EventGraph {nodes [
   Event BlueprintUpdateAnimation()
   > n1 = Sequence()
   > n2 = Gravity Override State Function DEPRECATED(CharacterPartAnimInstance=Self-Reference(), GameTimeInSeconds=GetGameTimeInSeconds(), GravityOverrideParams=Get Gravity Override Params())
   > Set Gravity_Override(Gravity_Override:Vector=@n2) -> (Output_Get:Vector)
   n3 = Get CapsuleComponent!(self:Character Object Reference=Get Pawn Owner() -> (Pawn Owner:Player Pawn Athena Object Reference)) -> (CapsuleComponent:Capsule Collision Object Reference)
   n4 = Unrotate Vector!(
    A:Vector=GetComponentVelocity!(self:Scene Component Object Reference=@n3) -> (ReturnValue:Vector)
    B:Rotator=Get World Rotation!(self:Scene Component Object Reference=@n3) -> (ReturnValue:Rotator) {is=title K2Node_CallFunction}
   ) -> (ReturnValue_X:double, ReturnValue_Y:double, ReturnValue_Z:double)
   Set AcceleratingForward(execute=@n1.then_1, AcceleratingForward:double=SelectFloat(A=1, B=0, bPickA="float >= float"(A=@n4, B=100))) -> (Output_Get:double)
   Event OnCharacterCustomizationLoaded()
   n5 = CheckForBackbling()
   > Set isWearingPack(isWearingPack:bool=@n5) -> (Output_Get:bool)
   > Set isWearingCape(isWearingCape=AND Boolean(A=@n5, B=@n5.bIsBackblingLong))
   Event BlueprintInitializeAnimation()
   > Set BodyMesh(BodyMesh=GetOwnerSkeletalMeshComponent())
   > n6 = Cast To PlayerPawn_Athena(Object=TryGetPawnOwner())
   > Set Pawn Owner(Pawn Owner:Player Pawn Athena Object Reference=@n6) -> (Output_Get:Player Pawn Athena Object Reference)
  ]}
]
variables [
 AcceleratingForward {type=number}
 BodyMesh {type=SkeletalMeshComponent}
 Gravity Override Params {
  type=GravityOverrideParamsStruct
  default {
   skydive_FloatingDown {gravityOverride=<0 0 820>, windAmplitude=<200 -500 300>, windOffset=<0 -600 100>}
   skydive_GoingUp {gravityOverride=<0 -300 500>, windAmplitude=<200 -500 300>}
   skydive_Parachuting {gravityOverride=<0 0 100>, windAmplitude=<200 -150 400>, windFrequency=<1.5 4 6>}
  }
 }
 Gravity_Override {type=Vector}
 isWearingCape {type=boolean}
 isWearingPack {type=boolean}
 NewVar_0 {type=PawnMovementComponent, default=null}
 Pawn Owner {type=PlayerPawn_Athena_C}
]
